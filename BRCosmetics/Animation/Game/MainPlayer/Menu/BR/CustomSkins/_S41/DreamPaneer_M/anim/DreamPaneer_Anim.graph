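# Level Sequence for animating the DreamPaneer_M custom skin in the Battle Royale menu, at 30 fps.
# The central spawned mannequin carries the MR_Biped_v02 body Control Rig track, with two sections, and a Fortnite_3L_Face_CtrlRig track for the face.
# Three more mannequins, offset along X and yawed to roughly side and back angles, each play a 473-frame skeletal animation, apparently as turnaround previews.
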LevelSequence /BRCosmetics/Animation/Game/MainPlayer/Menu/BR/CustomSkins/_S41/DreamPaneer_M/anim/DreamPaneer_Anim
bindings [
 mannequin2 {
  id=#1
  possessed_class=FortPlayerMannequin
  tracks [
   {type=MovieSceneSpawnTrack, sections [{type=MovieSceneSpawnSection, channels [None {type=MovieSceneScriptingBoolChannel, default=true}]}]}
   {
    type=MovieScene3DTransformTrack
    sections [{
      type=MovieScene3DTransformSection
      channels [
       Location.X {type=MovieSceneScriptingDoubleChannel, default=-84.4415}
       Location.Y {type=MovieSceneScriptingDoubleChannel, default=0}
       Location.Z {type=MovieSceneScriptingDoubleChannel, default=0}
       Rotation.X {type=MovieSceneScriptingDoubleChannel, default=0}
       Rotation.Y {type=MovieSceneScriptingDoubleChannel, default=0}
       Rotation.Z {type=MovieSceneScriptingDoubleChannel, default=79.1146}
       Scale.X {type=MovieSceneScriptingDoubleChannel, default=1}
       Scale.Y {type=MovieSceneScriptingDoubleChannel, default=1}
       Scale.Z {type=MovieSceneScriptingDoubleChannel, default=1}
      ]
     }]
   }
   {type=MovieSceneSkeletalAnimationTrack, sections [{type=MovieSceneSkeletalAnimationSection, end_frame=473, start_frame=0}]}
  ]
 }
 mannequin {
  id=#2
  possessed_class=FortPlayerMannequin
  tracks [
   {type=MovieSceneSpawnTrack, sections [{type=MovieSceneSpawnSection, channels [None {type=MovieSceneScriptingBoolChannel, default=true}]}]}
   {
    type=MovieScene3DTransformTrack
    sections [{
      type=MovieScene3DTransformSection
      channels [
       Location.X {type=MovieSceneScriptingDoubleChannel, default=0}
       Location.Y {type=MovieSceneScriptingDoubleChannel, default=0}
       Location.Z {type=MovieSceneScriptingDoubleChannel, default=0}
       Rotation.X {type=MovieSceneScriptingDoubleChannel, default=0}
       Rotation.Y {type=MovieSceneScriptingDoubleChannel, default=0}
       Rotation.Z {type=MovieSceneScriptingDoubleChannel, default=0}
       Scale.X {type=MovieSceneScriptingDoubleChannel, default=1}
       Scale.Y {type=MovieSceneScriptingDoubleChannel, default=1}
       Scale.Z {type=MovieSceneScriptingDoubleChannel, default=1}
      ]
     }]
   }
   {
    type=MovieSceneControlRigParameterTrack
    display_name=MR_Biped_v02
    sections [{type=MovieSceneControlRigParameterSection}, {type=MovieSceneControlRigParameterSection}]
   }
   {
    type=MovieSceneControlRigParameterTrack
    display_name=Fortnite_3L_Face_CtrlRig
    sections [{
      type=MovieSceneControlRigParameterSection
      channels [
       CTRL_C_eyesAim.Location.X {type=MovieSceneScriptingFloatChannel, default=0}
       CTRL_C_eyesAim.Location.Y {type=MovieSceneScriptingFloatChannel, default=0}
       CTRL_C_eyesAim.Location.Z {type=MovieSceneScriptingFloatChannel, default=0}
       CTRL_C_eyesAim.Rotation.X {type=MovieSceneScriptingFloatChannel, default=0}
       CTRL_C_eyesAim.Rotation.Y {type=MovieSceneScriptingFloatChannel, default=0}
       CTRL_C_eyesAim.Rotation.Z {type=MovieSceneScriptingFloatChannel, default=0}
       CTRL_C_eyesAim.Scale.X {type=MovieSceneScriptingFloatChannel, default=1}
       CTRL_C_eyesAim.Scale.Y {type=MovieSceneScriptingFloatChannel, default=1}
       CTRL_C_eyesAim.Scale.Z {type=MovieSceneScriptingFloatChannel, default=1}
       CTRL_L_eyeAim.Location.X {type=MovieSceneScriptingFloatChannel, default=0}
       CTRL_L_eyeAim.Location.Y {type=MovieSceneScriptingFloatChannel, default=0}
       CTRL_L_eyeAim.Location.Z {type=MovieSceneScriptingFloatChannel, default=0}
       CTRL_L_eyeAim.Rotation.X {type=MovieSceneScriptingFloatChannel, default=0}
       CTRL_L_eyeAim.Rotation.Y {type=MovieSceneScriptingFloatChannel, default=0}
       CTRL_L_eyeAim.Rotation.Z {type=MovieSceneScriptingFloatChannel, default=0}
       CTRL_L_eyeAim.Scale.X {type=MovieSceneScriptingFloatChannel, default=1}
       CTRL_L_eyeAim.Scale.Y {type=MovieSceneScriptingFloatChannel, default=1}
       CTRL_L_eyeAim.Scale.Z {type=MovieSceneScriptingFloatChannel, default=1}
       CTRL_R_eyeAim.Location.X {type=MovieSceneScriptingFloatChannel, default=0}
       CTRL_R_eyeAim.Location.Y {type=MovieSceneScriptingFloatChannel, default=0}
       CTRL_R_eyeAim.Location.Z {type=MovieSceneScriptingFloatChannel, default=0}
       CTRL_R_eyeAim.Rotation.X {type=MovieSceneScriptingFloatChannel, default=0}
       CTRL_R_eyeAim.Rotation.Y {type=MovieSceneScriptingFloatChannel, default=0}
       CTRL_R_eyeAim.Rotation.Z {type=MovieSceneScriptingFloatChannel, default=0}
       CTRL_R_eyeAim.Scale.X {type=MovieSceneScriptingFloatChannel, default=1}
       CTRL_R_eyeAim.Scale.Y {type=MovieSceneScriptingFloatChannel, default=1}
       CTRL_R_eyeAim.Scale.Z {type=MovieSceneScriptingFloatChannel, default=1}
       CTRL_convergenceSwitch {type=MovieSceneScriptingFloatChannel, default=0}
       CTRL_Master.Location.X {type=MovieSceneScriptingFloatChannel, default=0}
       CTRL_Master.Location.Y {type=MovieSceneScriptingFloatChannel, default=0}
       CTRL_Master.Location.Z {type=MovieSceneScriptingFloatChannel, default=0}
       CTRL_Master.Rotation.X {type=MovieSceneScriptingFloatChannel, default=0}
       CTRL_Master.Rotation.Y {type=MovieSceneScriptingFloatChannel, default=0}
       CTRL_Master.Rotation.Z {type=MovieSceneScriptingFloatChannel, default=0}
       CTRL_Master.Scale.X {type=MovieSceneScriptingFloatChannel, default=1}
       CTRL_Master.Scale.Y {type=MovieSceneScriptingFloatChannel, default=1}
       CTRL_Master.Scale.Z {type=MovieSceneScriptingFloatChannel, default=1}
       CTRL_C_jaw.X {type=MovieSceneScriptingFloatChannel, default=0}
       CTRL_C_jaw.Y {type=MovieSceneScriptingFloatChannel, default=0}
       CTRL_C_jaw_openExtreme {type=MovieSceneScriptingFloatChannel, default=0}
       CTRL_C_jaw_fwdBack {type=MovieSceneScriptingFloatChannel, default=0}
       CTRL_R_jaw_ChinRaiseD {type=MovieSceneScriptingFloatChannel, default=0}
       CTRL_R_jaw_ChinRaiseU {type=MovieSceneScriptingFloatChannel, default=0}
       CTRL_L_jaw_ChinRaiseU {type=MovieSceneScriptingFloatChannel, default=0}
       CTRL_L_jaw_ChinRaiseD {type=MovieSceneScriptingFloatChannel, default=0}
       CTRL_R_mouth_lipsBlow {type=MovieSceneScriptingFloatChannel, default=0}
       CTRL_L_mouth_lipsTogetherU {type=MovieSceneScriptingFloatChannel, default=0}
       CTRL_L_mouth_lipsTogetherD {type=MovieSceneScriptingFloatChannel, default=0}
       CTRL_R_mouth_lipsTogetherU {type=MovieSceneScriptingFloatChannel, default=0}
       CTRL_L_mouth_lipsBlow {type=MovieSceneScriptingFloatChannel, default=0}
       CTRL_R_mouth_lipsTogetherD {type=MovieSceneScriptingFloatChannel, default=0}
       CTRL_C_tongue_roll.X {type=MovieSceneScriptingFloatChannel, default=0}
       CTRL_C_tongue_roll.Y {type=MovieSceneScriptingFloatChannel, default=0}
       CTRL_C_tongue.X {type=MovieSceneScriptingFloatChannel, default=0}
       CTRL_C_tongue.Y {type=MovieSceneScriptingFloatChannel, default=0}
       CTRL_C_tongue_narrowWide {type=MovieSceneScriptingFloatChannel, default=0}
       CTRL_C_tongue_inOut {type=MovieSceneScriptingFloatChannel, default=0}
       CTRL_R_nose.X {type=MovieSceneScriptingFloatChannel, default=0}
       CTRL_R_nose.Y {type=MovieSceneScriptingFloatChannel, default=0}
       CTRL_L_nose.X {type=MovieSceneScriptingFloatChannel, default=0}
       CTRL_L_nose.Y {type=MovieSceneScriptingFloatChannel, default=0}
       CTRL_L_eye.X {type=MovieSceneScriptingFloatChannel, default=0}
       CTRL_L_eye.Y {type=MovieSceneScriptingFloatChannel, default=0}
       CTRL_L_eye_blink {type=MovieSceneScriptingFloatChannel, default=0}
       CTRL_R_eye.X {type=MovieSceneScriptingFloatChannel, default=0}
       CTRL_R_eye.Y {type=MovieSceneScriptingFloatChannel, default=0}
       CTRL_R_eye_blink {type=MovieSceneScriptingFloatChannel, default=0}
       CTRL_C_eye.X {type=MovieSceneScriptingFloatChannel, default=0}
       CTRL_C_eye.Y {type=MovieSceneScriptingFloatChannel, default=0}
       CTRL_C_eyesConvergence {type=MovieSceneScriptingFloatChannel, default=0}
       CTRL_L_eye_cheekRaise {type=MovieSceneScriptingFloatChannel, default=0}
       CTRL_L_eye_squintInner {type=MovieSceneScriptingFloatChannel, default=0}
       CTRL_L_nose_wrinkleUpper {type=MovieSceneScriptingFloatChannel, default=0}
       CTRL_R_nose_wrinkleUpper {type=MovieSceneScriptingFloatChannel, default=0}
       CTRL_R_eye_squintInner {type=MovieSceneScriptingFloatChannel, default=0}
       CTRL_R_eye_cheekRaise {type=MovieSceneScriptingFloatChannel, default=0}
       CTRL_C_eyelidsReduce {type=MovieSceneScriptingFloatChannel, default=0}
       CTRL_R_brow_lateral {type=MovieSceneScriptingFloatChannel, default=0}
       CTRL_L_brow_lateral {type=MovieSceneScriptingFloatChannel, default=0}
       CTRL_R_brow_down {type=MovieSceneScriptingFloatChannel, default=0}
       CTRL_L_brow_down {type=MovieSceneScriptingFloatChannel, default=0}
       CTRL_L_brow_raiseOut {type=MovieSceneScriptingFloatChannel, default=0}
       CTRL_L_brow_raiseIn {type=MovieSceneScriptingFloatChannel, default=0}
       CTRL_R_brow_raiseIn {type=MovieSceneScriptingFloatChannel, default=0}
       CTRL_R_brow_raiseOut {type=MovieSceneScriptingFloatChannel, default=0}
       CTRL_lookAtSwitch {type=MovieSceneScriptingFloatChannel, default=0}
       CTRL_C_teethD.X {type=MovieSceneScriptingFloatChannel, default=0}
       CTRL_C_teethD.Y {type=MovieSceneScriptingFloatChannel, default=0}
       CTRL_C_teethU.X {type=MovieSceneScriptingFloatChannel, default=0}
       CTRL_C_teethU.Y {type=MovieSceneScriptingFloatChannel, default=0}
       CTRL_C_teeth_fwdBackU {type=MovieSceneScriptingFloatChannel, default=0}
       CTRL_L_mouth_corner.X {type=MovieSceneScriptingFloatChannel, default=0}
       CTRL_L_mouth_corner.Y {type=MovieSceneScriptingFloatChannel, default=0}
       CTRL_R_mouth_corner.X {type=MovieSceneScriptingFloatChannel, default=0}
       CTRL_R_mouth_corner.Y {type=MovieSceneScriptingFloatChannel, default=0}
       CTRL_C_teeth_fwdBackD {type=MovieSceneScriptingFloatChannel, default=0}
       CTRL_R_mouth_cornerSharpnessD {type=MovieSceneScriptingFloatChannel, default=0}
       CTRL_R_mouth_cornerSharpnessU {type=MovieSceneScriptingFloatChannel, default=0}
       CTRL_L_eye_eyelidU.X {type=MovieSceneScriptingFloatChannel, default=0}
       CTRL_L_eye_eyelidU.Y {type=MovieSceneScriptingFloatChannel, default=0}
       CTRL_L_mouth_cornerSharpnessD {type=MovieSceneScriptingFloatChannel, default=0}
       CTRL_L_eye_eyelidD.X {type=MovieSceneScriptingFloatChannel, default=0}
       CTRL_L_eye_eyelidD.Y {type=MovieSceneScriptingFloatChannel, default=0}
       CTRL_L_mouth_cornerSharpnessU {type=MovieSceneScriptingFloatChannel, default=0}
       CTRL_R_mouth_lipsTowardsTeethD {type=MovieSceneScriptingFloatChannel, default=0}
       CTRL_L_mouth_lipsTowardsTeethD {type=MovieSceneScriptingFloatChannel, default=0}
       CTRL_R_eye_eyelidU.X {type=MovieSceneScriptingFloatChannel, default=0}
       CTRL_R_eye_eyelidU.Y {type=MovieSceneScriptingFloatChannel, default=0}
       CTRL_L_mouth_lipsTowardsTeethU {type=MovieSceneScriptingFloatChannel, default=0}
       CTRL_R_eye_eyelidD.X {type=MovieSceneScriptingFloatChannel, default=0}
       CTRL_R_eye_eyelidD.Y {type=MovieSceneScriptingFloatChannel, default=0}
       CTRL_R_mouth_lipsTowardsTeethU {type=MovieSceneScriptingFloatChannel, default=0}
       CTRL_faceGUIfollowHead {type=MovieSceneScriptingFloatChannel, default=1}
       CTRL_eyesAimFollowHead {type=MovieSceneScriptingFloatChannel, default=0}
       CTRL_C_nose.X {type=MovieSceneScriptingFloatChannel, default=0}
       CTRL_C_nose.Y {type=MovieSceneScriptingFloatChannel, default=0}
       CTRL_C_teeth_rotateU {type=MovieSceneScriptingFloatChannel, default=0}
       CTRL_C_teeth_rotateD {type=MovieSceneScriptingFloatChannel, default=0}
       CTRL_L_brow_inner {type=MovieSceneScriptingFloatChannel, default=0}
       CTRL_R_brow_inner {type=MovieSceneScriptingFloatChannel, default=0}
       CTRL_L_brow_outer {type=MovieSceneScriptingFloatChannel, default=0}
       CTRL_R_brow_outer {type=MovieSceneScriptingFloatChannel, default=0}
       CTRL_C_mouth.X {type=MovieSceneScriptingFloatChannel, default=0}
       CTRL_C_mouth.Y {type=MovieSceneScriptingFloatChannel, default=0}
       CTRL_L_mouth_upperLipRaise {type=MovieSceneScriptingFloatChannel, default=0}
       CTRL_L_mouth_cornerPull {type=MovieSceneScriptingFloatChannel, default=0}
       CTRL_L_mouth_dimple {type=MovieSceneScriptingFloatChannel, default=0}
       CTRL_L_mouth_cornerDepress {type=MovieSceneScriptingFloatChannel, default=0}
       CTRL_L_mouth_stretch {type=MovieSceneScriptingFloatChannel, default=0}
       CTRL_L_mouth_lowerLipDepress {type=MovieSceneScriptingFloatChannel, default=0}
       CTRL_R_mouth_lowerLipDepress {type=MovieSceneScriptingFloatChannel, default=0}
       CTRL_R_mouth_stretch {type=MovieSceneScriptingFloatChannel, default=0}
       CTRL_L_mouth_towardsD {type=MovieSceneScriptingFloatChannel, default=0}
       CTRL_R_mouth_funnelU {type=MovieSceneScriptingFloatChannel, default=0}
       CTRL_L_mouth_towardsU {type=MovieSceneScriptingFloatChannel, default=0}
       CTRL_R_mouth_lipBiteD {type=MovieSceneScriptingFloatChannel, default=0}
       CTRL_R_mouth_funnelD {type=MovieSceneScriptingFloatChannel, default=0}
       CTRL_L_mouth_pressU {type=MovieSceneScriptingFloatChannel, default=0}
       CTRL_L_mouth_funnelD {type=MovieSceneScriptingFloatChannel, default=0}
       CTRL_L_mouth_suckBlow {type=MovieSceneScriptingFloatChannel, default=0}
       CTRL_L_mouth_pressD {type=MovieSceneScriptingFloatChannel, default=0}
       CTRL_L_mouth_funnelU {type=MovieSceneScriptingFloatChannel, default=0}
       CTRL_R_mouth_towardsU {type=MovieSceneScriptingFloatChannel, default=0}
       CTRL_R_mouth_purseD {type=MovieSceneScriptingFloatChannel, default=0}
       CTRL_L_mouth_lipBiteU {type=MovieSceneScriptingFloatChannel, default=0}
       CTRL_R_mouth_pressD {type=MovieSceneScriptingFloatChannel, default=0}
       CTRL_R_mouth_pressU {type=MovieSceneScriptingFloatChannel, default=0}
       CTRL_L_mouth_purseD {type=MovieSceneScriptingFloatChannel, default=0}
       CTRL_R_mouth_lipBiteU {type=MovieSceneScriptingFloatChannel, default=0}
       CTRL_R_mouth_suckBlow {type=MovieSceneScriptingFloatChannel, default=0}
       CTRL_L_mouth_purseU {type=MovieSceneScriptingFloatChannel, default=0}
       CTRL_R_mouth_towardsD {type=MovieSceneScriptingFloatChannel, default=0}
       CTRL_L_mouth_lipBiteD {type=MovieSceneScriptingFloatChannel, default=0}
       CTRL_R_mouth_purseU {type=MovieSceneScriptingFloatChannel, default=0}
       CTRL_R_mouth_cornerDepress {type=MovieSceneScriptingFloatChannel, default=0}
       CTRL_R_mouth_upperLipRaise {type=MovieSceneScriptingFloatChannel, default=0}
       CTRL_R_mouth_dimple {type=MovieSceneScriptingFloatChannel, default=0}
       CTRL_R_mouth_cornerPull {type=MovieSceneScriptingFloatChannel, default=0}
      ]
     }]
   }
  ]
 }
 mannequin3 {
  id=#3
  possessed_class=FortPlayerMannequin
  tracks [
   {type=MovieSceneSpawnTrack, sections [{type=MovieSceneSpawnSection, channels [None {type=MovieSceneScriptingBoolChannel, default=true}]}]}
   {
    type=MovieScene3DTransformTrack
    sections [{
      type=MovieScene3DTransformSection
      channels [
       Location.X {type=MovieSceneScriptingDoubleChannel, default=87.685}
       Location.Y {type=MovieSceneScriptingDoubleChannel, default=0}
       Location.Z {type=MovieSceneScriptingDoubleChannel, default=0}
       Rotation.X {type=MovieSceneScriptingDoubleChannel, default=0}
       Rotation.Y {type=MovieSceneScriptingDoubleChannel, default=0}
       Rotation.Z {type=MovieSceneScriptingDoubleChannel, default=-81.1734}
       Scale.X {type=MovieSceneScriptingDoubleChannel, default=1}
       Scale.Y {type=MovieSceneScriptingDoubleChannel, default=1}
       Scale.Z {type=MovieSceneScriptingDoubleChannel, default=1}
      ]
     }]
   }
   {type=MovieSceneSkeletalAnimationTrack, sections [{type=MovieSceneSkeletalAnimationSection, end_frame=473, start_frame=0}]}
  ]
 }
 mannequin4 {
  id=#4
  possessed_class=FortPlayerMannequin
  tracks [
   {type=MovieSceneSpawnTrack, sections [{type=MovieSceneSpawnSection, channels [None {type=MovieSceneScriptingBoolChannel, default=true}]}]}
   {
    type=MovieScene3DTransformTrack
    sections [{
      type=MovieScene3DTransformSection
      channels [
       Location.X {type=MovieSceneScriptingDoubleChannel, default=157.6287}
       Location.Y {type=MovieSceneScriptingDoubleChannel, default=0}
       Location.Z {type=MovieSceneScriptingDoubleChannel, default=0}
       Rotation.X {type=MovieSceneScriptingDoubleChannel, default=0}
       Rotation.Y {type=MovieSceneScriptingDoubleChannel, default=0}
       Rotation.Z {type=MovieSceneScriptingDoubleChannel, default=180}
       Scale.X {type=MovieSceneScriptingDoubleChannel, default=1}
       Scale.Y {type=MovieSceneScriptingDoubleChannel, default=1}
       Scale.Z {type=MovieSceneScriptingDoubleChannel, default=1}
      ]
     }]
   }
   {type=MovieSceneSkeletalAnimationTrack, sections [{type=MovieSceneSkeletalAnimationSection, end_frame=473, start_frame=0}]}
  ]
 }
]
display_rate {denominator=1, numerator=30}
tick_resolution {denominator=1, numerator=24000}
tracks []
